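# Niagara candle flame effect for a LightCastle centerpiece set, gated to nighttime through an Isnighttime spawn probability on every emitter.
# Flame and Smoke draw rising, curl-noise ribbons, with the first Flame particle's position and age held to anchor the ribbon at its base.
# Glow adds short-lived sprites with alternate halo color and size attributes.
# Light pairs a sprite with a light renderer, modulating color by a waveform for flicker.
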NiagaraSystem /Game/Athena/Items/ForagedItems/Rift/FX/Candle/NS_LightCastle_CenterPieceSet_B_CandleFlame
dependencies {
 usedDataInterfaces [&NiagaraDataInterfaceCurve, &NiagaraDataInterfaceColorCurve]
 usedDynamicInputs [&FN_Select_Float_from_Time_Of_Day]
 usedModules [
  &SystemState
  &self:SystemSpawnScript.SystemScriptSource.SystemScriptGraph.NiagaraNodeAssignment_0.SetVariables_#1
 ]
}
emitters {
 Flame {
  emitterUpdateScript [
   &EmitterState(Life Cycle Mode=::NewEnumerator1, Loop Duration=1, UseLoopDelay=false)
   &SpawnBurst_Instantaneous(Spawn Probability=$System.Isnighttime, Use Spawn Probability=true, UseLoopCountLimit=true)
   &SpawnRate(Spawn Probability=$System.Isnighttime, SpawnRate=7, Use Spawn Probability=true)
  ]
  particleSpawnScript [
   &InitializeRibbon(
    Color=<rgba 10 1.75 0.2 0.3>
    Lifetime=&Multiply_Float(
     A=&Lerp_Float(
      A=0.75
      B=0.85
      Alpha=&FN_LOD_Distance_Remap(Start Distance=500, End Distance=1500, Minimum Output=0, Maximum Output=1)
     )
     B=1.25
    )
    Module.Write Color=false
    Module.Write Lifetime=true
    Module.Write Mass=true
    Module.Write Material Random=true
    Module.Write Position=true
    Module.Write RibbonTwist=false
    Module.Write RibbonWidth=true
    Ribbon Width=&Multiply_Float(
     A=&Lerp_Float(
      A=3
      B=3.5
      Alpha=&FN_LOD_Distance_Remap(Start Distance=500, End Distance=1500, Minimum Output=0, Maximum Output=1)
     )
     B=2
    )
   ) {enabled=false}
   &InitializeParticle(
    Lifetime=&Multiply_Float(
     A=&Lerp_Float(
      A=0.75
      B=0.85
      Alpha=&FN_LOD_Distance_Remap(Start Distance=500, End Distance=1500, Minimum Output=0, Maximum Output=1)
     )
     B=1.25
    )
    Position Offset=<0 0 -1.5>
    Position Offset Coordinate Space=::World
    Ribbon Width=&Multiply_Float(
     A=&Lerp_Float(
      A=3
      B=3.5
      Alpha=&FN_LOD_Distance_Remap(Start Distance=500, End Distance=1500, Minimum Output=0, Maximum Output=1)
     )
     B=2
    )
    Ribbon Width Mode=::NewEnumerator1
    Sprite Size Mode=::NewEnumerator3
    Uniform Sprite Size=1
    UsePositionOffset=true
    Write Material Random=true
   )
  ]
  particleUpdateScript [
   &ParticleState()
   &GravityForce(Gravity=<0 0 60>)
   /Niagara/Modules/Update/Forces/CurlNoiseForce(
    Noise Frequency=0.01
    Noise Strength=&FloatFromCurve(
     FloatCurve=DataInterface() {propertyValues {Curve {keys [
         {arriveTangent=0.1097, interpMode=RCIM_Cubic, leaveTangent=0.1097, tangentMode=RCTM_User, time=0, value=0}
         {arriveTangent=1.6798, interpMode=RCIM_Cubic, leaveTangent=1.6798, tangentMode=RCTM_User, time=1, value=1}
        ]}}}
     Scale Curve=30
    )
    Pan Noise=false
    Random Seed:NiagaraInt32=&RandomRangeInt(Minimum=0, Maximum=100)
   )
   &Drag(Drag=2.5, Use Linear Drag=true, Use Rotational Drag=false)
   &SolveForcesAndVelocity(Clamp Velocity=false, Limit Acceleration=false)
   &Color(Color=&ColorFromCurve(ColorCurve=DataInterface() {propertyValues {
       AlphaCurve {keys [<at 0 1 RCIM_Cubic>]}
       BlueCurve {keys [<at 0 10 RCIM_Cubic>,<at 0.1638 0.2 RCIM_Cubic>]}
       GreenCurve {keys [<at 0 1.75 RCIM_Cubic>,<at 0.1638 1.75 RCIM_Cubic>]}
       RedCurve {keys [<at 0 0.2 RCIM_Cubic>,<at 0.1638 10 RCIM_Cubic>]}
      }}))
   &ScaleColor(
    Color Value To Scale=$Particles.Color
    Scale Alpha=&FloatFromCurve(
     FloatCurve=DataInterface() {propertyValues {Curve {keys [
         {arriveTangent=-0.0176, interpMode=RCIM_Cubic, leaveTangent=-0.0176, tangentMode=RCTM_User, time=0.1, value=0}
         {arriveTangent=0.1593, interpMode=RCIM_Cubic, leaveTangent=0.1593, tangentMode=RCTM_User, time=0.3, value=1}
         {arriveTangent=-0.0405, interpMode=RCIM_Cubic, leaveTangent=-0.0405, tangentMode=RCTM_User, time=1, value=0}
        ]}}}
     Scale Curve=0.5
    )
    ScaleA=true
    ScaleRGB=true
   )
   &ScaleRibbonWidth(Ribbon Width Scale=&FloatFromCurve(
     FloatCurve=DataInterface() {propertyValues {Curve {keys [
         {arriveTangent=0.0742, interpMode=RCIM_Cubic, leaveTangent=0.0742, tangentMode=RCTM_User, time=0, value=0.4}
         {arriveTangent=-0.0613, interpMode=RCIM_Cubic, leaveTangent=-0.0613, tangentMode=RCTM_User, time=0.1408, value=0.9847}
         {arriveTangent=-0.0405, interpMode=RCIM_Cubic, leaveTangent=-0.0405, tangentMode=RCTM_User, time=1, value=0}
        ]}}}
     Scale Curve=0.5
    ))
   /SetVariables(Particles.Position=&MakeCustomVectorFromBool(
     Bool=&SetBoolByIntComparison(A=&ReturnParticleIDAttributes(), B=0, Comparison Type=::NewEnumerator4)
     True Vector=Linked() {linkedVariable {name=Particles.Initial.Position, type=NiagaraPosition}}
     False Vector=Linked() {linkedVariable {name=Particles.Position, type=NiagaraPosition}}
    )) {isSetParameters=true}
   /SetVariables(Particles.NormalizedAge:NiagaraFloat=&MakeCustomFloatFromBool(
     Bool=&SetBoolByIntComparison(A=&ReturnParticleIDAttributes(), B=0, Comparison Type=::NewEnumerator4)
     True Float=0
     False Float=$Particles.NormalizedAge
    )) {enabled=false, isSetParameters=true}
   /SetVariables(Particles.Age:NiagaraFloat=&MakeCustomFloatFromBool(
     Bool=&SetBoolByIntComparison(A=&ReturnParticleIDAttributes(), B=0, Comparison Type=::NewEnumerator4)
     True Float=0
     False Float=$Particles.Age
    )) {isSetParameters=true}
   /SetVariables(Particles.RibbonLinkOrder:NiagaraFloat=$Particles.NormalizedAge) {isSetParameters=true}
  ]
  rendererClasses [&NiagaraRibbonRendererProperties, &NiagaraSpriteRendererProperties]
  renderers [
   {rendererClass=&NiagaraRibbonRendererProperties}
   {rendererClass=&NiagaraSpriteRendererProperties, rendererIndex=1}
  ]
 }
 Glow {
  emitterUpdateScript [
   &EmitterState(Life Cycle Mode=::NewEnumerator1, Loop Duration=1, UseLoopDelay=false)
   &SpawnRate(Spawn Probability=$System.Isnighttime, SpawnRate=10, Use Spawn Probability=true)
  ]
  particleSpawnScript [
   /Niagara/Modules/Spawn/Initialization/InitializeParticle(
    Color=<rgba 15 10 1 1>
    Lifetime=&Multiply_Float(A=&UniformRangedFloat(Minimum=0.3, Maximum=0.6), B=1)
    Module.Write Color=true
    Module.Write Lifetime=true
    Module.Write Mass=true
    Module.Write Material Random=true
    Module.Write Position=true
    Module.Write Scale=false
    Module.Write SpriteRotation=false
    Module.Write SpriteSize=true
    Position=&SimulationPosition()
    Sprite Size=&Vector2DFromFloat(Value=&Multiply_Float(A=&UniformRangedFloat(Minimum=0.5, Maximum=0.7), B=3))
   )
   &Color(Color=&MultiplyLinearColor(
     Linear Color=$Particles.Color
     Scale Factor=/Niagara/DynamicInputs/UniformRange/UniformRangedVector4(
      Minimum:Vector4f=<1 0.25 0.25 0.01>
      Maximum:Vector4f=<1 1 1 1>
      Randomness Mode:ENiagaraRandomnessMode=::NewEnumerator0
      Random Seed:NiagaraInt32=0
      Fixed Random Seed:NiagaraBool=false
     )
    ))
  ]
  particleUpdateScript [
   &ParticleState()
   &AddVelocity(Coordinate Space=::World, Velocity=&RandomRangeVector(Minimum=<0 0 2>, Maximum=<0 0 4>))
   &CurlNoiseForce(
    Noise Frequency=0.01
    Noise Strength=&FloatFromCurve(
     FloatCurve=DataInterface() {propertyValues {Curve {keys [
         {arriveTangent=0.1097, interpMode=RCIM_Cubic, leaveTangent=0.1097, tangentMode=RCTM_User, time=0, value=0}
         {arriveTangent=1.6798, interpMode=RCIM_Cubic, leaveTangent=1.6798, tangentMode=RCTM_User, time=1, value=1}
        ]}}}
     Scale Curve=30
    )
    Pan Noise=false
    Random Seed:NiagaraInt32=&RandomRangeInt(Minimum=0, Maximum=100)
   )
   &Drag(Drag=30, Use Linear Drag=true, Use Rotational Drag=false)
   &SolveForcesAndVelocity(Clamp Velocity=false, Limit Acceleration=false)
   &ScaleColor(
    Scale Alpha=&FloatFromCurve(FloatCurve=DataInterface() {propertyValues {Curve {keys [<at 0.5 1>,<at 1 0>]}}})
    Scale RGB=&Lerp_Vector(A=<1 1 1>, B=<1 0 0>, Alpha=$Particles.NormalizedAge)
    ScaleA=true
    ScaleRGB=true
   )
   &ScaleSpriteSize(Scale Factor=&Vector2DFromFloat(Value=&FloatFromCurve(FloatCurve=DataInterface() {propertyValues {Curve {keys [
          {arriveTangent=-0.0036, interpMode=RCIM_Cubic, leaveTangent=-0.0036, tangentMode=RCTM_User, time=0, value=1}
          {arriveTangent=-2.2201, interpMode=RCIM_Cubic, leaveTangent=-2.2201, tangentMode=RCTM_User, time=1, value=0}
         ]}}})))
   /SetVariables(
    Particles.ColorAlt:LinearColor=&MultiplyLinearColor(Linear Color=$Particles.Color, Scale Factor=<1 0.35 0.1 0.02>)
    Particles.SpriteSizeAlt:Vector2f=&Vector2DFromFloat(Value=&FloatFromCurve(
      FloatCurve=DataInterface() {propertyValues {Curve {keys [{arriveTangent=-0.0036, interpMode=RCIM_Cubic, leaveTangent=-0.0036, tangentMode=RCTM_User, time=0, value=1}]}}}
      Scale Curve=&RandomRangeFloat(Minimum=3, Maximum=4.5)
     ))
   ) {isSetParameters=true}
   &CameraOffset(Camera Offset Amount=5)
  ]
  rendererClasses [&NiagaraSpriteRendererProperties]
  renderers [
   {rendererClass=&NiagaraSpriteRendererProperties}
   {rendererClass=&NiagaraSpriteRendererProperties, rendererIndex=1}
  ]
 }
 Light {
  emitterUpdateScript [
   &EmitterState(Life Cycle Mode=::NewEnumerator1, Loop Duration=1, UseLoopDelay=false)
   &SpawnBurst_Instantaneous(Spawn Probability=$System.Isnighttime, Use Spawn Probability=true, UseLoopCountLimit=false)
  ]
  particleSpawnScript [/Niagara/Modules/Spawn/Initialization/V2/InitializeParticle(Color=<rgba 3 0.525 0.06 1>, Sprite Size Mode=::NewEnumerator3, Uniform Sprite Size=1, UsePositionOffset=false, Write Material Random=true)]
  particleUpdateScript [
   &ParticleState()
   &FN_Light_Attributes(Create Exponent Attribute=true, Create Radius Attribute=true, Exponent=2, Radius=80)
   &ScaleColor(
    Scale RGB=&VectorFromFloat(Value=&Waveform("Waveform [1]"=::NewEnumerator12, "[1] Phase Scale"=10, "[1] Frequency"=1, "[1] Amplitude Min/Max"=<0 1>, "[1] Bias"=1))
    ScaleA=true
    ScaleRGB=true
   )
  ]
  rendererClasses [&NiagaraSpriteRendererProperties, &NiagaraLightRendererProperties]
  renderers [
   {rendererClass=&NiagaraSpriteRendererProperties}
   {rendererClass=&NiagaraLightRendererProperties, rendererIndex=1}
  ]
 }
 Smoke {
  emitterUpdateScript [
   &EmitterState(Life Cycle Mode=::NewEnumerator1, Loop Duration=1, UseLoopDelay=false)
   &SpawnRate(Spawn Probability=$System.Isnighttime, SpawnRate=7, Use Spawn Probability=true)
  ]
  particleSpawnScript [&InitializeRibbon(
    Color=<rgba 10 1.75 0.2 1>
    Lifetime=1.5
    Module.Write Color=true
    Module.Write Lifetime=true
    Module.Write Mass=true
    Module.Write Material Random=true
    Module.Write Position=true
    Module.Write RibbonTwist=false
    Module.Write RibbonWidth=true
    Position:Vector3f=&Add_Vector(A=/Niagara/DynamicInputs/Helpers/SimulationPosition(), B=<0 0 8>)
    Ribbon Width=&Multiply_Float(A=&RandomRangeFloat(Minimum=2.5, Maximum=3.5), B=1.25)
   )]
  particleUpdateScript [
   &ParticleState()
   &GravityForce(Gravity=<0 0 75>)
   &CurlNoiseForce(
    Noise Frequency=0.01
    Noise Strength=&FloatFromCurve(
     FloatCurve=DataInterface() {propertyValues {Curve {keys [
         {arriveTangent=1.0245, interpMode=RCIM_Cubic, leaveTangent=1.0245, tangentMode=RCTM_User, time=0, value=0}
         {arriveTangent=1.6798, interpMode=RCIM_Cubic, leaveTangent=1.6798, tangentMode=RCTM_User, time=1, value=1}
        ]}}}
     Scale Curve=30
    )
    Pan Noise=false
    Random Seed:NiagaraInt32=&RandomRangeInt(Minimum=0, Maximum=100)
   )
   &Drag(Drag=2.5, Use Linear Drag=true, Use Rotational Drag=false) {module=Drag001}
   &FN_RibbonMaterial(Color Pan U Speed=-0.5, Erosion Amount=1, Macro Distortion Amount=1, Opacity Erosion=0.25, Opacity Pan U Speed=-0.5, Opacity Power=0.35)
   &ScaleRibbonWidth(Ribbon Width Scale=&FloatFromCurve(
     FloatCurve=DataInterface() {propertyValues {Curve {keys [
         {arriveTangent=0.0219, interpMode=RCIM_Cubic, leaveTangent=0.0219, tangentMode=RCTM_User, time=0, value=0}
         <at 0.5 0.9666 RCIM_Cubic>
         {arriveTangent=-0.2116, interpMode=RCIM_Cubic, leaveTangent=-0.2116, tangentMode=RCTM_User, time=1, value=0}
        ]}}}
     Scale Curve=&RandomRangeFloat(Minimum=1, Maximum=1)
    )) {module=ScaleRibbonWidth001}
   &SolveForcesAndVelocity(Clamp Velocity=false, Limit Acceleration=false)
   &ScaleColor(
    Scale Alpha=&FloatFromCurve(
     FloatCurve=DataInterface() {propertyValues {Curve {keys [
         {arriveTangent=9.0725, interpMode=RCIM_Cubic, leaveTangent=9.0725, tangentMode=RCTM_User, time=0, value=0}
         {arriveTangent=0.0486, interpMode=RCIM_Cubic, leaveTangent=0.0486, tangentMode=RCTM_User, time=0.1408, value=0.9847}
         {arriveTangent=-0.0405, interpMode=RCIM_Cubic, leaveTangent=-0.0405, tangentMode=RCTM_User, time=1, value=0}
        ]}}}
     Scale Curve=1.25
    )
    ScaleA=true
    ScaleRGB=true
   )
  ]
  rendererClasses [&NiagaraRibbonRendererProperties]
  renderers [{rendererClass=&NiagaraRibbonRendererProperties}]
 }
}
summary {
 systemName=NS_LightCastle_CenterPieceSet_B_CandleFlame
 emitters [
  {emitterName=Glow, rendererClasses [&NiagaraSpriteRendererProperties]}
  {emitterName=Flame, rendererClasses [&NiagaraRibbonRendererProperties, &NiagaraSpriteRendererProperties]}
  {emitterName=Smoke, rendererClasses [&NiagaraRibbonRendererProperties]}
  {emitterName=Light, rendererClasses [&NiagaraSpriteRendererProperties, &NiagaraLightRendererProperties]}
 ]
 userVariables []
}
